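# Post-process material in RemoteSession, named for showing an AR camera feed.
# Its only sampler is a 2D texture parameter, CameraImage, which lets the image be swapped on a material instance.
# The sampler reads with default texture coordinates, and the graph lists no connection from it to a material output.
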Material /RemoteSession/ARCameraPostProcess
graph {
 properties {materialDomain=MD_PostProcess, pixelDepthOffsetMode=PDOM_Legacy, usedShadingModels=Default Lit}
 nodes [TextureSampleParameter2D(UVs=TextureCoordinate()) {parameterName=CameraImage}]
}
